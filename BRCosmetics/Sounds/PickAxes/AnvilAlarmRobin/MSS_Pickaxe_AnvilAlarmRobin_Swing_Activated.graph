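# One-shot stereo MetaSound source for the activated swing sound of the AnvilAlarmRobin pickaxe cosmetic.
# On play, a random getter picks one of the reactive swing waves and feeds it to a stereo wave player.
# Both channels are scaled by a Volume input, and the player's finish trigger ends the source.
# Pickaxe attenuation, property sheet, melee sound class and concurrency assets govern playback.
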MetaSoundSource /BRCosmetics/Sounds/PickAxes/AnvilAlarmRobin/MSS_Pickaxe_AnvilAlarmRobin_Swing_Activated
inputs [
 Sounds {type=WaveAsset:Array}
 UE.Source.OnPlay {type=Trigger}
 Volume {type=Float}
]
interfaces [UE.OutputFormat.Stereo:1.0, UE.Source.OneShot:1.0, UE.Source:1.1]
outputs [
 "UE.OutputFormat.Stereo.Audio:0" {type=Audio}
 "UE.OutputFormat.Stereo.Audio:1" {type=Audio}
 UE.Source.OneShot.OnFinished {type=Trigger}
]
pages [Default {
  input_defaults {
   UE.Source.OnPlay=null
   Volume=0.4
   Sounds [
    &/Pickaxe_AnvilAlarmRobin_Swing_Reactive_01
    &/Pickaxe_AnvilAlarmRobin_Swing_Reactive_02
    &/Pickaxe_AnvilAlarmRobin_Swing_Reactive_03
    &/Pickaxe_AnvilAlarmRobin_Swing_Reactive_04
    &/Pickaxe_AnvilAlarmRobin_Swing_Reactive_05
    &/Pickaxe_AnvilAlarmRobin_Swing_Reactive_06
   ]
  }
  nodes [
   ArrayRandomGet_0 = Array.Random Get:WaveAsset:Array(In Array=^Sounds, Next=^UE.Source.OnPlay) {also {defaults {No Repeats=-1}}}
   WavePlayer_0 = UE.Wave Player:Stereo(Play=@ArrayRandomGet_0.On Next, Wave Asset=@ArrayRandomGet_0.Value) {also {version [1,1]}}
   Multiply_0 = UE.Multiply:Audio by Float(AdditionalOperands=^Volume, PrimaryOperand=@WavePlayer_0.Out Left)
   Multiply_1 = UE.Multiply:Audio by Float(AdditionalOperands=^Volume, PrimaryOperand=@WavePlayer_0.Out Right)
  ]
  links [
   {dst="@UE.OutputFormat.Stereo.Audio:0", src=Multiply_0.Out}
   {dst="@UE.OutputFormat.Stereo.Audio:1", src=Multiply_1.Out}
   {dst="@UE.Source.OneShot.OnFinished", src=WavePlayer_0.On Finished}
  ]
 }]
properties {
 assetImportData=&self:AssetImportData
 attenuationSettings=/Game/Sounds/Attenuation/ATT_Fort_Weapon_Pickaxe_NoAtt_Att
 audioPropertiesSheet=/BRCosmetics/Sounds/Settings/PropertySheets/Pickaxe/APS_Cosmetic_Pickaxe_Swing
 maxDistance=3612
 numChannels=2
 outputFormat=Stereo
 sampleRate=48000
 soundClassObject=&Fort_WeaponMelee_PSM
 concurrencySet [&Player_Melee_Weapon_Conc]
}
